# Copyright Epic Games, Inc. All Rights Reserved.
#################################################
# Generated Digest of Verse API
# DO NOT modify this manually!
# Generated from build: ++Fortnite+Release-36.30-CL-44367537
#################################################


using {/Verse.org/Native}
using {/Verse.org/SceneGraph}
using {/Verse.org/Simulation}
using {/Verse.org/Colors}
using {/Verse.org/Assets}
# Module import path: /playkigai.dig@fortnite.com/Minigame_HorseDerby/Materials
Materials := module:
    TrackMaterial<scoped {/playkigai.dig@fortnite.com/Minigame_HorseDerby}> := class<final><scoped {/playkigai.dig@fortnite.com/Minigame_HorseDerby}>(material):
        @editable
        var EmissiveColor<public>:color = external {}

# Module import path: /playkigai.dig@fortnite.com/Minigame_HorseDerby/Meshes
Meshes := module:
    FinishLine<scoped {/playkigai.dig@fortnite.com/Minigame_HorseDerby}> := class<final><scoped {/playkigai.dig@fortnite.com/Minigame_HorseDerby}>(mesh_component):
        @editable
        var Material_0<public>:material = external {}

    FinishLine_asset<scoped {/playkigai.dig@fortnite.com/Minigame_HorseDerby}>:mesh = external {}

    GameManagerMesh<scoped {/playkigai.dig@fortnite.com/Minigame_HorseDerby}> := class<final><scoped {/playkigai.dig@fortnite.com/Minigame_HorseDerby}>(mesh_component):
        @editable
        var WorldGridMaterial<public>:material = external {}

    GameManagerMesh_asset<scoped {/playkigai.dig@fortnite.com/Minigame_HorseDerby}>:mesh = external {}

    HorseBackLegs<scoped {/playkigai.dig@fortnite.com/Minigame_HorseDerby}> := class<final><scoped {/playkigai.dig@fortnite.com/Minigame_HorseDerby}>(mesh_component):
        @editable
        var WorldGridMaterial<public>:material = external {}

    HorseBackLegs_asset<scoped {/playkigai.dig@fortnite.com/Minigame_HorseDerby}>:mesh = external {}

    HorseBody<scoped {/playkigai.dig@fortnite.com/Minigame_HorseDerby}> := class<final><scoped {/playkigai.dig@fortnite.com/Minigame_HorseDerby}>(mesh_component):
        @editable
        var Material_0<public>:material = external {}

    HorseBody_asset<scoped {/playkigai.dig@fortnite.com/Minigame_HorseDerby}>:mesh = external {}

    HorseFrontLegs<scoped {/playkigai.dig@fortnite.com/Minigame_HorseDerby}> := class<final><scoped {/playkigai.dig@fortnite.com/Minigame_HorseDerby}>(mesh_component):
        @editable
        var WorldGridMaterial<public>:material = external {}

    HorseFrontLegs_asset<scoped {/playkigai.dig@fortnite.com/Minigame_HorseDerby}>:mesh = external {}

    HorseHead<scoped {/playkigai.dig@fortnite.com/Minigame_HorseDerby}> := class<final><scoped {/playkigai.dig@fortnite.com/Minigame_HorseDerby}>(mesh_component):
        @editable
        var WorldGridMaterial<public>:material = external {}

    HorseHead_asset<scoped {/playkigai.dig@fortnite.com/Minigame_HorseDerby}>:mesh = external {}

    ObstacleHigh<scoped {/playkigai.dig@fortnite.com/Minigame_HorseDerby}> := class<final><scoped {/playkigai.dig@fortnite.com/Minigame_HorseDerby}>(mesh_component):
        @editable
        var WorldGridMaterial<public>:material = external {}

    ObstacleHigh_asset<scoped {/playkigai.dig@fortnite.com/Minigame_HorseDerby}>:mesh = external {}

    ObstacleLow<scoped {/playkigai.dig@fortnite.com/Minigame_HorseDerby}> := class<final><scoped {/playkigai.dig@fortnite.com/Minigame_HorseDerby}>(mesh_component):
        @editable
        var WorldGridMaterial<public>:material = external {}

    ObstacleLow_asset<scoped {/playkigai.dig@fortnite.com/Minigame_HorseDerby}>:mesh = external {}

    RaceTrack<scoped {/playkigai.dig@fortnite.com/Minigame_HorseDerby}> := class<final><scoped {/playkigai.dig@fortnite.com/Minigame_HorseDerby}>(mesh_component):
        @editable
        var Material<public>:material = external {}

    RaceTrack_asset<scoped {/playkigai.dig@fortnite.com/Minigame_HorseDerby}>:mesh = external {}

    StartLine<scoped {/playkigai.dig@fortnite.com/Minigame_HorseDerby}> := class<final><scoped {/playkigai.dig@fortnite.com/Minigame_HorseDerby}>(mesh_component):
        @editable
        var Material_0<public>:material = external {}

    StartLine_asset<scoped {/playkigai.dig@fortnite.com/Minigame_HorseDerby}>:mesh = external {}

# Module import path: /playkigai.dig@fortnite.com/Minigame_HorseDerby/Prefabs
Prefabs := module:
    @import_as("/Minigame_HorseDerby/Prefabs/game_manager.game_manager_C")
    game_manager<scoped {/playkigai.dig@fortnite.com/Minigame_HorseDerby}> := class<final>(entity):


    game_manager_asset<scoped {/playkigai.dig@fortnite.com/Minigame_HorseDerby}>:entity_prefab = external {}

    @import_as("/Minigame_HorseDerby/Prefabs/horse.horse_C")
    horse<scoped {/playkigai.dig@fortnite.com/Minigame_HorseDerby}> := class<final>(entity):


    horse_asset<scoped {/playkigai.dig@fortnite.com/Minigame_HorseDerby}>:entity_prefab = external {}

    @import_as("/Minigame_HorseDerby/Prefabs/horse_head.horse_head_C")
    horse_head<scoped {/playkigai.dig@fortnite.com/Minigame_HorseDerby}> := class<final>(entity):


    horse_head_asset<scoped {/playkigai.dig@fortnite.com/Minigame_HorseDerby}>:entity_prefab = external {}

    @import_as("/Minigame_HorseDerby/Prefabs/horse_legs_b.horse_legs_b_C")
    horse_legs_b<scoped {/playkigai.dig@fortnite.com/Minigame_HorseDerby}> := class<final>(entity):


    horse_legs_b_asset<scoped {/playkigai.dig@fortnite.com/Minigame_HorseDerby}>:entity_prefab = external {}

    @import_as("/Minigame_HorseDerby/Prefabs/horse_legs_f.horse_legs_f_C")
    horse_legs_f<scoped {/playkigai.dig@fortnite.com/Minigame_HorseDerby}> := class<final>(entity):


    horse_legs_f_asset<scoped {/playkigai.dig@fortnite.com/Minigame_HorseDerby}>:entity_prefab = external {}

    @import_as("/Minigame_HorseDerby/Prefabs/obstacle_high.obstacle_high_C")
    obstacle_high<scoped {/playkigai.dig@fortnite.com/Minigame_HorseDerby}> := class<final>(entity):


    obstacle_high_asset<scoped {/playkigai.dig@fortnite.com/Minigame_HorseDerby}>:entity_prefab = external {}

    @import_as("/Minigame_HorseDerby/Prefabs/obstacle_low.obstacle_low_C")
    obstacle_low<scoped {/playkigai.dig@fortnite.com/Minigame_HorseDerby}> := class<final>(entity):


    obstacle_low_asset<scoped {/playkigai.dig@fortnite.com/Minigame_HorseDerby}>:entity_prefab = external {}

    @import_as("/Minigame_HorseDerby/Prefabs/race_finish_line.race_finish_line_C")
    race_finish_line<scoped {/playkigai.dig@fortnite.com/Minigame_HorseDerby}> := class<final>(entity):


    race_finish_line_asset<scoped {/playkigai.dig@fortnite.com/Minigame_HorseDerby}>:entity_prefab = external {}

    @import_as("/Minigame_HorseDerby/Prefabs/race_start_line.race_start_line_C")
    race_start_line<scoped {/playkigai.dig@fortnite.com/Minigame_HorseDerby}> := class<final>(entity):


    race_start_line_asset<scoped {/playkigai.dig@fortnite.com/Minigame_HorseDerby}>:entity_prefab = external {}

    @import_as("/Minigame_HorseDerby/Prefabs/race_track.race_track_C")
    race_track<scoped {/playkigai.dig@fortnite.com/Minigame_HorseDerby}> := class<final>(entity):


    race_track_asset<scoped {/playkigai.dig@fortnite.com/Minigame_HorseDerby}>:entity_prefab = external {}