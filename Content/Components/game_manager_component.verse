<#
@UUID: 9c80b91f-f754-43fd-8b57-cf2fb90168e0
@Task: Game Manager
@Description: `game_manager` starts, stops, resets, and manages round-state for Horse Derby.
@Tags: GameManager
@Export: true
#>
<#
Entity Prefab setup for `game_manager` entity prefab:
    The `game_manager` is the central logic to the minigame.

    - game_manager
        - mesh_component
            NOTE:
                Mesh *must* have collision (e.g. `BlockAllDynamic`).
        - transform_component
        - game_manager_component
#>

using { /Fortnite.com/Devices }
using { /Fortnite.com/Playspaces }
using { /Verse.org/Random }
using { /Verse.org/Colors }
using { /Verse.org/Colors/NamedColors }
using { /Verse.org/SceneGraph }
using { /Verse.org/Simulation }
using { /Verse.org/SpatialMath }

using { EventsModule }
using { HorseModule }
using { Prefabs }
using { ObstacleModule }
using { PriorityQueue }
using { RaceTrackModule }

GameManagerModule<public> := module:

    game_states<public> := enum:
        AwaitingStart,
        QueueCountdown,
        Loading,
        InProgress,
        Finished


    # Class for managing the game.
    game_manager_component<public> := class<final_super>(interactable_component):
        @editable { Categories := array{InputTriggersCatText} }
        JumpInput: input_trigger_device = input_trigger_device{}
        @editable { Categories := array{InputTriggersCatText} }
        CrouchInput: input_trigger_device = input_trigger_device{}

        @editable { Categories := array{GameOptionsCatText} }
        RaceOrigin: vector3 = vector3{}
        @editable { Categories := array{GameOptionsCatText} }
        RaceDistance: float = 2500.0

        @editable { Categories := array{GameOptionsCatText} }
        QueueTime: int = 3
        @editable { Categories := array{GameOptionsCatText} }
        MaxPlayerCount: int = 4
        @editable { Categories := array{GameOptionsCatText} }
        MinPlayerCount: int = 1


        var MaybeEventsBus: ?event_bus = false
        var GameState: game_states = game_states.AwaitingStart

        var PlayerNumQueue: priority_queue(int) = priority_queue(int){}
        var AgentToMaybePlayerNum: [agent]?int = map{}
        var QueueTimer: int = 0
        var WinnerCheckTime: float = 0.2
        var WinnerCheckTimer: float = 0.0
        var Winners: []agent = array{}

        var RaceTracks: []race_track = array{}
        var Horses: []horse = array{}

        PlayerColors: []color = array{
            NamedColors.Aqua,
            NamedColors.Crimson,
            NamedColors.LawnGreen,
            NamedColors.LemonChiffon,
            NamedColors.Orchid
        }

        # Buffer distance for obstacle spawning on game start.
        ObstacleSpawnDistance: float = 250.0
        ObstacleSpawnDelay: float = 1.15


        # === Entity Functions ===

        OnBeginSimulation<override>():void=
            # === Events ===
            set MaybeEventsBus = option{GetEventBus[]}
            JumpInput.PressedEvent.Subscribe(OnJumpPressed)
            CrouchInput.PressedEvent.Subscribe(OnCrouchPressed)
            SucceededEvent.Subscribe(OnInteractedWith)

            if (Playspace := Entity.GetPlayspaceForEntity[]):
                Playspace.PlayerRemovedEvent().Subscribe(OnPlayerRemoved)

            # === Player Numbers Queue ===
            PlayerNumberArray := for (Num := 1..MaxPlayerCount) do Num
            set PlayerNumQueue = PlayerNumQueue.Initialize(PlayerNumberArray, IntGreater)


        # Handle the PlayerRemoved event.
        OnPlayerRemoved(InPlayer:player):void=
            if:
                InAgent := agent[InPlayer]
                PlayerNum := AgentToMaybePlayerNum[InAgent]?
                set AgentToMaybePlayerNum[InAgent] = false
            then:
                set PlayerNumQueue = PlayerNumQueue.Push(PlayerNum, IntGreater)


        OnSimulate<override>()<suspends>:void=
            if (Events := MaybeEventsBus?):
                sync:
                    loop:
                        OnPlayerQueued(Events)

                    loop:
                        OnPlayerDequeued(Events)

                    loop:
                        Sleep(ObstacleSpawnDelay)
                        if (GameState = game_states.InProgress):
                            SpawnRandomObstacle()


        # === Game Manager Functions ===

        # Handle the JumpPressed event.
        OnJumpPressed(InAgent:agent):void=
            if (EventsBus := GetEventBus[]):
                EventsBus.JumpPressed.Signal(InAgent)


        # Handle the CrouchPressed event.
        OnCrouchPressed(InAgent:agent):void=
            if (EventsBus := GetEventBus[]):
                EventsBus.CrouchPressed.Signal(InAgent)


        # Handle the PlayerQueue event.
        OnPlayerQueued(Events:event_bus)<suspends>:void=
            Events.PlayerQueued.Await()

            # Start queue timer if this is the first player queueing.
            if (PlayerNumQueue.Size() = (MaxPlayerCount - 1) and GameState = game_states.AwaitingStart):
                set GameState = game_states.QueueCountdown
                spawn { RunQueueCountdown() }
            else:
                # Set the queue timer back to QueueTime to allow others to join or leave.
                set QueueTimer = QueueTime


        # Run the queue countdown and start the game when it reaches 0.
        RunQueueCountdown()<suspends>:void=
            set QueueTimer = QueueTime

            loop:
                # If the game state is AwaitingStart after a countdown has been started,
                #   then the game has been canceled.
                if (GameState = game_states.AwaitingStart):
                    break

                if (GameState = game_states.InProgress):
                    break

                if (QueueTimer <> 0):
                    Print("Starting in {QueueTimer}")

                if (QueueTimer = 0):
                    InitializeRace()
                    Print("Start!!!")
                    break

                Sleep(1.0)
                set QueueTimer -= 1


        # Handle the PlayerDequeue event.
        OnPlayerDequeued(Events:event_bus)<suspends>:void=
            Events.PlayerDequeued.Await()

            if (PlayerNumQueue.Size() = MaxPlayerCount):
                Print("No players in queue, cancelling game.")
                set GameState = game_states.AwaitingStart


        # Handle the Succeeded event from this interactable component.
        OnInteractedWith(InAgent:agent):void=
            # End game.
            if:
                AgentToMaybePlayerNum[InAgent]?
                not GameState = game_states.AwaitingStart and not GameState = game_states.QueueCountdown
            then:
                set GameState = game_states.AwaitingStart
                Reset()
                return
            
            # Leave queue.
            if:
                PlayerNum := AgentToMaybePlayerNum[InAgent]?
                set AgentToMaybePlayerNum[InAgent] = false

                EventsBus := GetEventBus[]
            then:
                set PlayerNumQueue = PlayerNumQueue.Push(PlayerNum, IntGreater)

                EventsBus.PlayerDequeued.Signal()
                return
            
            # Join queue.
            if:
                not PlayerNumQueue.Empty[]

                set AgentToMaybePlayerNum[InAgent] = option{PlayerNumQueue.Top[]}
                NewPlayerQueue := PlayerNumQueue.Pop[IntGreater]

                EventsBus := GetEventBus[]
            then:
                set PlayerNumQueue = NewPlayerQueue

                EventsBus.PlayerQueued.Signal()
                return


        # Initialize the race for each player.
        InitializeRace():void=
            set GameState = game_states.Loading

            SpawnRaceTracks()

            for:
                Agent -> MaybePlayerNum : AgentToMaybePlayerNum
                PlayerNum := MaybePlayerNum?
            do:
                SpawnHorses(PlayerNum, Agent)

            set WinnerCheckTimer = WinnerCheckTime

            set GameState = game_states.InProgress


        # Spawn the race track for each player.
        SpawnRaceTracks():void=
            for:
                Agent -> MaybePlayerNum : AgentToMaybePlayerNum
                PlayerNum := MaybePlayerNum?
                PlayerColor := PlayerColors[PlayerNum]

                SimRoot   := Entity.GetSimulationEntity[]

                RaceTrackEntity := race_track{}
                RaceTrackComp   := RaceTrackEntity.GetComponent[race_track_component]
            do:
                StartLoc := CalculateStartLocation(PlayerNum - 1)
                EndLoc   := StartLoc + vector3{Forward := RaceDistance}

                RaceTrackComp.SetGameManagerRef(Self)
                RaceTrackComp.SetStartLoc(StartLoc)
                RaceTrackComp.SetEndLoc(EndLoc)
                RaceTrackComp.SetTrackColor(?MaybeInColor := option{PlayerColor})

                NewRotation := MakeShortestRotationBetween(RaceTrackEntity.GetGlobalTransform().Rotation.GetForwardAxis(), EndLoc - StartLoc) 
                RaceTrackEntity.SetGlobalTransform(transform{Translation := StartLoc, Rotation := NewRotation})

                SimRoot.AddEntities(array{RaceTrackEntity})
                set RaceTracks += array{RaceTrackEntity}


        # Spawn the horses for each player.
        SpawnHorses(InIndex:int, InAgent:agent):void=
            if:
                SimRoot   := Entity.GetSimulationEntity[]

                Horse     := horse{}
                HorseComp := Horse.GetComponent[horse_component]
            then:
                StartLoc := CalculateStartLocation(InIndex - 1)
                EndLoc   := StartLoc + vector3{Forward := RaceDistance}

                HorseComp.Initialize(option{Self}, option{InAgent}, StartLoc, EndLoc)

                NewRotation := MakeShortestRotationBetween(Horse.GetGlobalTransform().Rotation.GetForwardAxis(), EndLoc - StartLoc)
                Horse.SetGlobalTransform(transform{Translation := StartLoc, Rotation := NewRotation})

                SimRoot.AddEntities(array{Horse})
                set Horses += array{Horse}


        # Reset the game variables.
        Reset<public>():void=
            # Reset players.
            for:
                Key -> MaybeValue : AgentToMaybePlayerNum
                Value := MaybeValue?
                set AgentToMaybePlayerNum[Key] = false
            do:
                set PlayerNumQueue = PlayerNumQueue.Push(Value, IntGreater)

            # Reset race tracks.
            for (Track : RaceTracks):
                Track.ResetTrack()
                Track.RemoveFromParent()


            # Reset horses.
            for (Horse : Horses):
                Horse.RemoveFromParent()

            set RaceTracks = array{}
            set Horses     = array{}
            set Winners    = array{}
            set GameState  = game_states.AwaitingStart


        # Add a winner for the race.
        AddWinner<public>(InAgent:agent):void=
            set Winners += array{InAgent}

            if (not GameState = game_states.Finished):
                spawn { RunWinnerTimer() }

            set GameState = game_states.Finished


        # Async function that allows ties to resolve before declaring a winner.
        # Otherwise, ties would have a single winner essentially chosen at random.
        RunWinnerTimer()<suspends>:void=
            SleepTime: float = 0.02

            loop:
                if (WinnerCheckTimer <= 0.0, EventsBus := GetEventBus[]):
                    FinishRace()
                    break

                Sleep(SleepTime)
                set WinnerCheckTimer -= SleepTime


        # Finish the race and reset.
        FinishRace():void=
            NumWinners := Winners.Length
            Print("{NumWinners} winners!")

            Reset()


        # Get the starting location for track number InIndex.
        CalculateStartLocation<public>(InIndex:int):vector3=
            RaceOrigin + vector3{Forward := 700.0, Left := 200.0, Up := 100.0 + 250.0 * InIndex}


        # Randomly select an obstacle and spawn it on all tracks.
        SpawnRandomObstacle():void=
            RandomNum := GetRandomInt(0, ObstacleTypesArray.Length - 1)

            ObstacleType: obstacle_types =
                if (ChosenType := ObstacleTypesArray[RandomNum]):
                    ChosenType
                else:
                    obstacle_types.Low

            for (Track : RaceTracks):
                Track.AddObstacle(ObstacleType)


        # Functor for priority queue.
        IntGreater<private>(InInt1:int, InInt2:int)<computes><decides>:void=
            InInt1 > InInt2


    # === Messages ===
    InputTriggersCatText<localizes>:message = "Input Triggers"
    GameOptionsCatText<localizes>:message = "Game Options"
    AgentNameText<localizes>(InAgent:agent):message= "{InAgent}"
    ToString(Agent:agent):string= Localize(AgentNameText(Agent))


    # Return the current game state.
    (InGMComp:game_manager_component).GetGameState()<transacts>:game_states=
        InGMComp.GameState


    # Return the obstacle spawn delay.
    (InGMComp:game_manager_component).GetObstacleSpawnDelay()<transacts>:float=
        InGMComp.ObstacleSpawnDelay


# Module for handling custom events for the game.
EventsModule<public> := module:
    var EventBus<public>: weak_map(session, event_bus) = map{}

    GetEventBus<public>()<decides><transacts>:event_bus=
        EventBus[GetSession()] or (set EventBus[GetSession()] = event_bus{})

    event_bus<public> := class:
        JumpPressed<public>:    event(agent) = event(agent){}
        CrouchPressed<public>:  event(agent) = event(agent){}
        PlayerQueued<public>:   event() = event(){}
        PlayerDequeued<public>: event() = event(){}