<#
@UUID: c2f7c2fe-b8b4-4a21-a751-274b6ed81396
@Task: Make a component to control the player's horse during the minigame.
@Description: `horse` is an entity prefab that travels along the race track from the Start Point towards the Finish point, avoiding obstacles along the way.
@Tags: Horse, Controller, Collision, Overlap Hits
@Export: true
#>
<#
Entity prefab setup for `horse` prefab:
    The `horse` prefab is the object that the player is controlling during the minigame.

    - horse
        - transform_component
        - mesh_component
            NOTE:
                Mesh *must* have collision (e.g. `BlockAllDynamic`).
        - horse_component

        - horse_head
            - transform_component
            - mesh_component
                NOTE:
                    Mesh *must* have collision (e.g. `BlockAllDynamic`).
        
        - horse_legs_b
            - transform_component
            - mesh_component
                NOTE:
                    Mesh *must* have collision (e.g. `BlockAllDynamic`).

        - horse_legs_f
            - transform_component
            - mesh_component
                NOTE:
                    Mesh *must* have collision (e.g. `BlockAllDynamic`).
#>

using { /Verse.org/SceneGraph }
using { /Verse.org/Simulation }
using { /Verse.org/SpatialMath }

using { EventsModule }
using { GameManagerModule }
using { ObstacleModule }
using { Prefabs }
using { RaceTrackModule }

HorseModule := module:

    # Component that controls the logic of the horse.
    horse_component<public> := class<final_super>(component):
        @editable
        HorseSpeed: float = 100.0
        @editable
        JumpHeight: float = 50.0
        @editable
        CrouchHeight: float = -50.0

        # === References ===
        var MaybeGameManager: ?game_manager_component = false
        var EventsBus: ?event_bus = false
        
        var MaybePilot: ?agent = false

        # === Movement Variables ===
        var StartLocation: vector3 = vector3{}
        var EndLocation: vector3 = vector3{}
        var NewTransform: transform = transform{}
        var HeightOffset: float = 0.0

        # === Action Variables ===    
        var StunDelay: float = 1.0
        var ActionDelay: float = 0.75
        var ActionLockTimer: float = 0.0
        var IsActionLocked: logic = false
        var IsStunned: logic = false

        var LerpTimeElapsed: float = 0.0    

        var FrameDelta: float = 0.0

        var MaybePrePhysics: ?cancelable = false
        var MaybePostPhysics: ?cancelable = false

        # === Entity Functions ===

        OnBeginSimulation<override>():void=
            set MaybePrePhysics = option{TickEvents.PrePhysics.Subscribe(OnPrePhysics)}

            set EventsBus = option{GetEventBus[]}

            if (MeshComponent := Entity.GetComponent[mesh_component]):
                MeshComponent.EntityEnteredEvent.Subscribe(OnEntityEntered)


        OnPrePhysics(Delta:float):void=
            set FrameDelta = Delta

            if (ActionLockTimer > 0.0):
                set ActionLockTimer -= Delta
            else if (IsActionLocked?):
                set IsActionLocked = false
                set IsStunned = false
                set LerpTimeElapsed = 0.0
                set HeightOffset = 0.0
                set ActionLockTimer = 0.0

            # Calculate the next horse position.
            CurrentSpeed := 
                if (IsStunned?):
                    0.0
                else:
                    HorseSpeed

            Transform := Entity.GetGlobalTransform()
            var TargetLoc: vector3 = Transform.Translation + Transform.Rotation.GetForwardAxis() * Delta * CurrentSpeed

            set TargetLoc.Up = StartLocation.Up

            if (IsActionLocked?, HeightOffset <> 0):
                Alpha := LerpTimeElapsed / ActionDelay
                set TargetLoc.Up += JumpParabola(Alpha) * HeightOffset
                set LerpTimeElapsed += Delta

            set NewTransform = transform:
                Translation := TargetLoc
                Rotation := Transform.Rotation

            Entity.SetGlobalTransform(NewTransform)


        OnSimulate<override>()<suspends>:void=
            ListenForPlayerInputEvents()


        OnEndSimulation<override>():void=
            if (PrePhysics := MaybePrePhysics?) { PrePhysics.Cancel() }


        # === Horse Functions ===

        # Handle the EntityEntered event.
        OnEntityEntered(InEntity:entity):void=
            if:
                FinishLineComponent := InEntity.GetComponent[finish_line_component]
                GM := MaybeGameManager?
                Agent := MaybePilot?
            then:
                GM.AddWinner(Agent)

            else if (ObstacleComponent := InEntity.GetComponent[obstacle_component]):
                UpdateActionLockTimer(StunDelay)
                set HeightOffset = 0.0
                set IsStunned = true
                InEntity.DestroyObstacle()


        # Handle the JumpPressed event.
        HandleJumpPressed(Events:event_bus)<suspends>:void=
            InAgent := Events.JumpPressed.Await()

            if (MaybePilot? = InAgent, not IsActionLocked?):
                set HeightOffset = JumpHeight
                UpdateActionLockTimer(ActionDelay)


        # Handle the CrouchPressed event.
        HandleCrouchPressed(Events:event_bus)<suspends>:void=
            InAgent := Events.CrouchPressed.Await()

            if (MaybePilot? = InAgent, not IsActionLocked?):
                set HeightOffset = CrouchHeight
                UpdateActionLockTimer(ActionDelay)


        # Asyncronously listen for jump or crouch commands.
        ListenForPlayerInputEvents()<suspends>:void=
            if (Events := EventsBus?):
                sync:
                    loop:
                        HandleJumpPressed(Events)

                    loop:
                        HandleCrouchPressed(Events)


        # Initialize horse variables.
        Initialize<public>(InMaybeGameManager:?game_manager_component, InAgent:?agent, InStartLoc:vector3, InEndLoc:vector3):void=
            set MaybeGameManager = InMaybeGameManager
            set MaybePilot       = InAgent
            set StartLocation    = InStartLoc
            set EndLocation      = InEndLoc


        # Calculate the point of the jump parabola given Alpha.
        JumpParabola(Alpha:float):float=
            (1.0 - Pow((1.0 - (2.0 * Alpha)), 4.0))


        # Update the action lock timer, preventing jump/crouch spam.
        UpdateActionLockTimer(Delay:float):void=
            if (Delay > ActionLockTimer):
                set ActionLockTimer = Delay
                set IsActionLocked = true