<#
@UUID: 4c514a0f-53ed-40c2-bc66-e57d5e528e73
@Task: Create a race track obstacle that must be avoided
@Description: `obstacle` is an entity prefab that travels along the race track from the Finish point towards the Start point. If a horse collides with an `obstacle`, the `obstacle` is destroyed and the horse becomes stunned.
@Tags: Race Track, Obstacle, Collision, Overlap Hits
@Export: true
#>
<#
Entity prefab setup for `obstacle` prefab:
    There are two `obstacle` prefabs: `obstacle_high` and `obstacle_low`. 
    `obstacle_high` must be crouched under, while `obstacle_high` must be jumped over.

    - obstacle_high
        - transform_component
        - mesh_component
            NOTE:
                Mesh *must* have collision (e.g. `BlockAllDynamic`).
        - obstacle_component

    - obstacle_low
        - transform_component
        - mesh_component
            NOTE:
                Mesh *must* have collision (e.g. `BlockAllDynamic`).
        - obstacle_components        
#>

using { /Verse.org/SceneGraph }
using { /Verse.org/Simulation }
using { /Verse.org/SpatialMath }

ObstacleModule := module:

    # An enum for object types.
    obstacle_types<public> := enum:
        Low,
        High


    # An array of available obstacle types.
    ObstacleTypesArray<public>: []obstacle_types = array{
        obstacle_types.Low,
        obstacle_types.High
    }


    # Component that defines an object as an obstacle.
    obstacle_component<public> := class<final_super>(component):
        @editable
        ObstacleSpeed: float = 400.0

        var MaybePrePhysics: ?cancelable = false


        # === Entity Functions ===

        OnBeginSimulation<override>():void=
            set MaybePrePhysics = option{TickEvents.PrePhysics.Subscribe(OnPrePhysics)}
        

        OnPrePhysics(Delta:float):void=
            Transform := Entity.GetGlobalTransform()
            TargetLoc: vector3 = Transform.Translation - Transform.Rotation.GetForwardAxis() * Delta * ObstacleSpeed

            NewTransform := transform:
                Translation := TargetLoc
                Rotation := Transform.Rotation

            Entity.SetGlobalTransform(NewTransform)

        
        OnEndSimulation<override>():void=
            if (PrePhysics := MaybePrePhysics?):
                PrePhysics.Cancel()


    # === Obstacle Functions ===

    # Disable the mesh component for the obstacle.
    (Obstacle:entity).DestroyObstacle<public>():void=
        Obstacle.RemoveFromParent()