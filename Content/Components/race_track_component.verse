<#
@UUID: 27833f2a-1a46-4ddb-92c8-8e0f7551c175
@Task: Create a race track that the player's horse will run across.
@Description: `race_track` is an entity prefab that is in charge of managing the track obstacles, the start and finish lines, and the track.
@Tags: Race Track, Start Line, Finish Line, Track
@Export: true
#>
<#
Entity prefab setup for `race_track` prefab:
    The `race_track` prefab is spawned for each player, managing the obstacles the player has to avoid as well as the start and finish lines.
    The track's floor can be set to a specific color.

    - race_track
        - transform_component
        - race_track_component

        - race_finish_line
            - transform_component
            - mesh_component
                NOTE:
                    Mesh *must* have collision (e.g. `BlockAllDynamic`).
            - finish_line_component
        
        - race_start_line
            - transform_component
            - mesh_component
                NOTE:
                    Mesh *must* have collision (e.g. `BlockAllDynamic`).
            - start_line_component

        - track
            - transform_component
            - mesh_component
                NOTE:
                    Mesh *must* have collision (e.g. `BlockAllDynamic`).
            - track_component
#>

using { /Verse.org }
using { /Verse.org/Colors }
using { /Verse.org/Native }
using { /Verse.org/Random }
using { /Verse.org/SceneGraph }
using { /Verse.org/Simulation }
using { /Verse.org/SpatialMath }

using { GameManagerModule }
using { ObstacleModule }
using { Prefabs }

RaceTrackModule<public> := module:

    # Component that manages the race track and its obstacles.
    race_track_component<public> := class<final_super>(component):
        var MaybeGameManager: ?game_manager_component = false

        # A buffer distance to spawn the obstacles behind the finish line.
        ObstacleSpawnBuffer: float = 1000.0

        BaseWidth: float = 100.0

        var StartLoc: vector3 = vector3{}
        var EndLoc: vector3 = vector3{}
        var RaceDistance: float = 0.0
        var Obstacles: []entity = array{}

        var TrackColor: color = color{}


        # === Entity Functions ===

        OnBeginSimulation<override>():void=
            # Scale the race track and set up finish line.
            TrackComponentsGenerator := Entity.FindDescendantEntitiesWithComponent(track_component)
            TrackEntityArray := for (TrackEntity : TrackComponentsGenerator) do TrackEntity

            FinishLineComponentsGenerator := Entity.FindDescendantEntitiesWithComponent(finish_line_component)
            FinishLineEntityArray := for (FinishLineEntity : FinishLineComponentsGenerator) do FinishLineEntity

            if:
                TrackEntity      := TrackEntityArray[0]
                TrackComp        := TrackEntity.GetComponent[track_component]
                FinishLineEntity := FinishLineEntityArray[0]
                FinishLineComp   := FinishLineEntity.GetComponent[mesh_component]
            then:
                set RaceDistance = EndLoc.Forward - StartLoc.Forward
                ScaleFactor := (RaceDistance - BaseWidth) / 100.0

                CurrentTransform := Entity.GetGlobalTransform()

                ScaleTransform := transform:
                    Translation := CurrentTransform.Translation
                    Rotation    := CurrentTransform.Rotation
                    Scale       := CurrentTransform.Scale + vector3{Forward := ScaleFactor}

                # TrackEntity.
                TrackEntity.SetGlobalTransform(ScaleTransform)
                TrackComp.SetColor(TrackColor)

                # Finish line.
                FinishLineEntity.SetGlobalTransform(transform{Translation := EndLoc})
                FinishLineComp.Enable()
            else:
                ErrorMsg<localizes>: message = "Error initializing race_track_component. Descendant entities not found."
                Print(ErrorMsg)


    # === Race Track Functions ===

    # Reset the obstacle track.
    (InRaceTrack:race_track).ResetTrack<public>():void=
        if (RaceTrackComp := InRaceTrack.GetComponent[race_track_component]):
            for (Obstacle : RaceTrackComp.Obstacles):
                Obstacle.DestroyObstacle()

            set RaceTrackComp.Obstacles = array{}


    # Set the GameManager reference.
    (InRaceTrack:race_track_component).SetGameManagerRef<public>(InGM:game_manager_component):void=
        set InRaceTrack.MaybeGameManager = option{InGM}


    # Set the StartLoc of the track component.
    (InRaceTrack:race_track_component).SetStartLoc<public>(InStartLoc:vector3):void=
        set InRaceTrack.StartLoc = InStartLoc


    # Set the EndLoc of the track component.
    (InRaceTrack:race_track_component).SetEndLoc<public>(InEndLoc:vector3):void=
        set InRaceTrack.EndLoc = InEndLoc


    # Get the StartLoc of the track component.
    (InRaceTrack:race_track_component).GetStartLoc<public>():vector3=
        InRaceTrack.StartLoc


    # Get the RaceDistance of the track component.
    (InRaceTrack:race_track_component).GetRaceDistance<public>():float=
        InRaceTrack.RaceDistance


    # Get the ObstacleSpawnBuffer of the track component.
    (InRaceTrack:race_track_component).GetObstacleSpawnBuffer<public>():float=
        InRaceTrack.ObstacleSpawnBuffer


    # Add an obstacle to the Obstacles array.
    (InRaceTrack:race_track_component).AddObstacleToArray<public>(InObstacle:entity):void=
        set InRaceTrack.Obstacles += array{InObstacle}


    # Wrap track_component's SetColor function.
    (InRaceTrack:race_track_component).SetTrackColor<public>(?MaybeInColor:?color = false):void=
        set InRaceTrack.TrackColor =
            if (InColor := MaybeInColor?):
                InColor
            else:
                color:
                    R := GetRandomFloat(0.0, 1.0)
                    G := GetRandomFloat(0.0, 1.0)
                    B := GetRandomFloat(0.0, 1.0)


    # Component that defines an object as the race track's finish line.
    finish_line_component<public> := class<final_super>(component):


    # Component that defines an object as the race track's start line.
    start_line_component<public> := class<final_super>(component):


    # Component for finding the track mesh for scaling.
    track_component<public> := class<final_super>(component):

        # Set the track's material to InColor.
        SetColor<public>(InColor:color):void=
            NewMat: Materials.TrackMaterial = Materials.TrackMaterial{}
            set NewMat.EmissiveColor = InColor

            if:
                TrackMeshComp := Entity.GetComponent[mesh_component]
                TrackMesh := Meshes.RaceTrack[TrackMeshComp]
            then:
                set TrackMesh.Material = NewMat


using { RaceTrackModule }

# Add an obstacle to the race track entity.
(InRaceTrackEntity:race_track).AddObstacle(InObstacleType:obstacle_types):void=

    var MaybeObstacle: ?entity = false
    case(InObstacleType):
        obstacle_types.Low  => set MaybeObstacle = option{obstacle_low{}}
        obstacle_types.High => set MaybeObstacle = option{obstacle_high{}}

    # Create a new obstacle and add it to the track.
    if:
        SimRoot       := InRaceTrackEntity.GetSimulationEntity[]
        RaceTrackComp := InRaceTrackEntity.GetComponent[RaceTrackModule.race_track_component]
        Obstacle      := MaybeObstacle?
        ObstacleComp  := Obstacle.GetComponent[obstacle_component]
    then:
        StartLoc    := RaceTrackComp.GetStartLoc()
        EndLoc      := StartLoc + vector3{Forward := RaceTrackComp.GetRaceDistance() + RaceTrackComp.GetObstacleSpawnBuffer()}
        NewRotation := Obstacle.GetGlobalTransform().Rotation
        Obstacle.SetGlobalTransform(transform{Translation := EndLoc, Rotation := NewRotation})

        RaceTrackComp.AddObstacleToArray(Obstacle)
        SimRoot.AddEntities(array{Obstacle})