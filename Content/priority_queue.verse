<#
@UUID: 216ea6cc-c7d1-4fc1-8f73-04ab9504f4fa
@Task: Implement a heap-based priority queue.
@Description: Implement a generalized heap-based priority queue that can accept a functor to determine which items are less than others.
@Tags: Algorithms, Data Structures, Priority Queue
@Export: true
#>
<#
Entity prefab setup for `obstacle` prefab:
    There are two `obstacle` prefabs: `obstacle_high` and `obstacle_low`. 
    `obstacle_high` must be crouched under, while `obstacle_high` must be jumped over.

    - obstacle_high
        - transform_component
        - mesh_component
            NOTE:
                Mesh *must* have collision (e.g. `BlockAllDynamic`).
        - obstacle_component

    - obstacle_low
        - transform_component
        - mesh_component
            NOTE:
                Mesh *must* have collision (e.g. `BlockAllDynamic`).
        - obstacle_components        
#>

PriorityQueue := module:

    # A heap-based priority queue.
    # Starts at index 0 of the array.
    # Left child is located at (2 * Index) + 1.
    # Right child is located at (2 * Index) + 2.
    # Parent is located Floor( (Index - 1) / 2 ).
    priority_queue<public>(t:type) := class<final>:
        # An array of elements, stored as a binary heap.
        Elements<public>: []t = array{}

        # Initializes the priority queue by adding one element at a time and calling FixUp on it.
        Initialize<public>(InitialElements:[]t, Less:type {_(:t, :t)<computes><decides>:void}):priority_queue(t)=
            var HeapedArray: []t = array{}
            
            for (Element : InitialElements):
                set HeapedArray += array{Element}
                set HeapedArray = HeapedArray.FixUp(HeapedArray.Length - 1, Less)

            priority_queue(t){ Elements := HeapedArray }


        # === Element Access ===

        # Get the top element of the priority queue.
        Top<public>()<decides><transacts>:t=
            Elements[0]


        # === Capacity ===

        # Succeeds if the priority queue is empty.
        # Fails if the priority queue is not empty.
        Empty<public>()<decides><transacts>:void=
            Size() = 0


        # Return the size of the priorty queue.
        Size<public>()<transacts>:int=
            Elements.Length


        # === Modifiers ===

        # Push NewElement, which is then sorted.
        # Call requires a functor of Less for the type of the priority queue.
        Push<public>(NewElement:t, Less:type {_(:t, :t)<computes><decides>:void}):priority_queue(t)=
            NewArray := Elements + array{NewElement}

            priority_queue(t){ Elements := NewArray.FixUp(NewArray.Length - 1, Less) }


        # Pop first element.
        Pop<public>(Less:type {_(:t, :t)<computes><decides>:void})<decides><transacts>:priority_queue(t)=
            # Set first element to last element, and delete last element.
            TempElement: t = Elements[Elements.Length - 1]
            
            ReplacedElementArr := Elements.ReplaceElement[0, TempElement]
            RemovedElementArr  := ReplacedElementArr.Remove[Elements.Length - 1, Elements.Length]

            # Sort first (previously last) element.
            priority_queue(t){ Elements := RemovedElementArr.FixDown(0, Less) }


    # Move element at InIndex up in the heap until it is at its sorted location.
    (InputArr:[]t where t:type).FixUp<public>(InIndex:int, Less:type {_(:t, :t)<computes><decides>:void})<transacts>:[]t=
        if (InIndex <= 0):
            return InputArr

        if:
            ParentIndex := Ceil(InIndex / 2) - 1

            ParentValue  := InputArr[ParentIndex]
            CurrentValue := InputArr[InIndex]

            Less[ParentValue, CurrentValue]

            # Swap parent and child values.
            TempArray := InputArr.ReplaceElement[ParentIndex, CurrentValue]
            NewArray  := TempArray.ReplaceElement[InIndex, ParentValue]
        then:
            return NewArray.FixUp(ParentIndex, Less)
        else:
            return InputArr


    # Move element at InIndex down in the heap until it is at its sorted location.
    (InputArr:[]t where t:type).FixDown<public>(InIndex:int, Less:type {_(:t, :t)<computes><decides>:void})<transacts>:[]t=
        if (InIndex >= InputArr.Length):
            return InputArr

        LeftChildIndex  := (2 * InIndex) + 1
        RightChildIndex := (2 * InIndex) + 2

        # Return if there are no children.
        if (LeftChildIndex >= InputArr.Length and RightChildIndex >= InputArr.Length):
            return InputArr

        # Initialize values.
        var LeftChildValue:  ?t = if (Val := InputArr[LeftChildIndex])  { option{Val} } else false
        var RightChildValue: ?t = if (Val := InputArr[RightChildIndex]) { option{Val} } else false
        var CurrentValue:    ?t = if (Val := InputArr[InIndex])         { option{Val} } else false

        # Both the left and right children are present.
        if (LeftChildValue? and RightChildValue?):
            # Check if right child has a higher priority.
            if (Less[LeftChildValue?, RightChildValue?]):
                if:
                    # Check if current value has a higher priority than the right child.
                    Less[CurrentValue?, RightChildValue?]

                    # Swap parent and child values.
                    TempArray := InputArr.ReplaceElement[InIndex, RightChildValue?]
                    NewArray  := TempArray.ReplaceElement[RightChildIndex, CurrentValue?]
                then:
                    return NewArray.FixDown(RightChildIndex, Less)
                else:
                    return InputArr
            
            # Left child has a higher or equal priority.
            else:
                if:
                    # Check if current value has a higher priority than the left child.
                    Less[CurrentValue?, LeftChildValue?]

                    # Swap parent and child values.
                    TempArray := InputArr.ReplaceElement[InIndex, LeftChildValue?]
                    NewArray  := TempArray.ReplaceElement[LeftChildIndex, CurrentValue?]
                then:
                    return NewArray.FixDown(LeftChildIndex, Less)
                else:
                    return InputArr

        # Only the left child is present.
        # Note: The RightChildValue is not checked due to the heap needing to be a complete binary tree.
        else if (LeftChildValue?):
            if:
                # Check if current value has a higher priority than the left child.
                Less[CurrentValue?, LeftChildValue?]

                # Swap parent and child values.
                TempArray := InputArr.ReplaceElement[InIndex, LeftChildValue?]
                NewArray  := TempArray.ReplaceElement[LeftChildIndex, CurrentValue?]
            then:
                return NewArray.FixDown(LeftChildIndex, Less)
            else:
                return InputArr

        # InIndex is a leaf node.
        else:
            return InputArr