<#
@UUID: 8cae8f7f-d051-480a-b1c9-e07a62e2fb8e
@Task: Test a heap-based priority queue written in Verse.
@Description: Test a heap-based priority queue written in Verse.
@Tags: Testing, Priority Queue, Algorithms, Data Structures
@Export: true
#>
<#
Test suite for `priority_queue` class.
#>

using { /Verse.org/SceneGraph }
using { /Verse.org/Simulation }

using { PriorityQueue }

PriorityQueueTests := module:

    # Test component for priority queue.
    # To run these, change the elements array in priority_queue.verse to <public>.
    priority_queue_tests_component<public> := class<final_super>(interactable_component):

        OnBeginSimulation<override>():void=
            SucceededEvent.Subscribe(OnInteractedWith)

        OnInteractedWith(InAgent:agent):void=
            ArrayAsc:  []int = array{1,2,3,4,5,6}
            ArrayDesc: []int = array{6,5,4,3,2,1}

            var MaxPQ: priority_queue(int) = priority_queue(int){}.Initialize(ArrayAsc, IntLess)
            var MinPQ: priority_queue(int) = priority_queue(int){}.Initialize(ArrayDesc, IntGreater)

            TestMaxInitialized(MaxPQ)
            TestMinInitialized(MinPQ)

            TestMaxTopAndPop(MaxPQ)
            TestMinTopAndPop(MinPQ)

            TestMaxPush()
            TestMinPush()

            TestPlayersLeavingAndJoining()


        # === Test Functions ===

        # Initialize test for Max Priority Queue.
        # Expected organization after adding {1,2,3,4,5,6} is {6,4,5,1,3,2}.
        TestMaxInitialized(MaxPQ:priority_queue(int)):void=
            var IsCorrect: logic = true

            if (not MaxPQ.Size() = 6):
                Print("MaxPQInitialized size incorrect.")
                set IsCorrect = false

            if:
                not MaxPQ.Elements[0] = 6
                not MaxPQ.Elements[1] = 4
                not MaxPQ.Elements[2] = 5
                not MaxPQ.Elements[3] = 1
                not MaxPQ.Elements[4] = 3
                not MaxPQ.Elements[5] = 2
            then:
                Print("MaxPQInitialized elements incorrect.")
                set IsCorrect = false

            if (MaxPQ.Empty[]):
                Print("MaxPQInitialized empty incorrect.")
                set IsCorrect = false
            
            TestName<localizes>:message = "TestMaxPQInitialized"
            PrintIsCorrectMessage(IsCorrect, TestName)


        # Test Top and Pop functions on MaxPQ. 
        # Expected top order is {6,5,4,3,2,1}.
        TestMaxTopAndPop(MaxPQ:priority_queue(int)):void=
            var IsCorrect: logic = true
            var MaxPQCopy: priority_queue(int) = MaxPQ

            # Checking the order of top after multiple pops.
            for:
                Index := 0..MaxPQ.Size()
                MaxPQEl := MaxPQCopy.Top[]
                NewMaxPQ := MaxPQCopy.Pop[IntLess]
            do:
                case(Index):
                    0 =>
                        if (not MaxPQEl = 6) {set IsCorrect = false}
                    1 => 
                        if (not MaxPQEl = 5) {set IsCorrect = false}
                    2 => 
                        if (not MaxPQEl = 4) {set IsCorrect = false}
                    3 => 
                        if (not MaxPQEl = 3) {set IsCorrect = false}
                    4 => 
                        if (not MaxPQEl = 2) {set IsCorrect = false}
                    5 => 
                        if (not MaxPQEl = 1) {set IsCorrect = false}
                    _ =>
                        Print("In default case for TestMaxPQTopAndPop. This should never happen.")

                set MaxPQCopy = NewMaxPQ
            
            TestName<localizes>:message = "TestMaxPQTopAndPop"
            PrintIsCorrectMessage(IsCorrect, TestName)
        
        
        # Test Push function on Max Priority Queue.
        # Order before push: {25,14,17,7,6,3,2}.
        # Expected order after push: {25,18,17,14,6,3,2,7}.
        TestMaxPush():void=
            Array := {25,14,17,7,6,3,2}
            var IsCorrect: logic = true
            var MaxPQ: priority_queue(int) = priority_queue(int){}.Initialize(Array, IntLess)

            set MaxPQ = MaxPQ.Push(18, IntLess)

            if (not MaxPQ.Size() = 8):
                Print("MaxPQPush size incorrect.")
                set IsCorrect = false

            if:
                not MaxPQ.Elements[0] = 25
                not MaxPQ.Elements[1] = 18
                not MaxPQ.Elements[2] = 17
                not MaxPQ.Elements[3] = 14
                not MaxPQ.Elements[4] = 6
                not MaxPQ.Elements[5] = 3
                not MaxPQ.Elements[6] = 2
                not MaxPQ.Elements[7] = 7
            then:
                Print("MaxPQ elements incorrect after push.")
                set IsCorrect = false

            if (MaxPQ.Empty[]):
                Print("MaxPQPush empty incorrect.")
                set IsCorrect = false
            
            TestName<localizes>:message = "TestMaxPQPush"
            PrintIsCorrectMessage(IsCorrect, TestName)


        # Initialize test for Min Priority Queue.
        # Expected organization after adding {6,5,4,3,2,1} is {1,3,2,6,4,5}.
        TestMinInitialized(MinPQ:priority_queue(int)):void=
            var IsCorrect: logic = true

            if (not MinPQ.Size() = 6):
                Print("MinPQInitialized size incorrect.")
                set IsCorrect = false

            if:
                not MinPQ.Elements[0] = 1
                not MinPQ.Elements[1] = 3
                not MinPQ.Elements[2] = 2
                not MinPQ.Elements[3] = 6
                not MinPQ.Elements[4] = 4
                not MinPQ.Elements[5] = 5
            then:
                Print("MinPQInitialized elements incorrect.")
                set IsCorrect = false

            if (MinPQ.Empty[]):
                Print("MinPQInitialized empty incorrect.")
                set IsCorrect = false
            
            TestName<localizes>:message = "TestMinPQInitialized"
            PrintIsCorrectMessage(IsCorrect, TestName)


        # Test Top and Pop functions on PQ1. 
        # Expected top order is {1,2,3,4,5,6}.
        TestMinTopAndPop(MinPQ:priority_queue(int)):void=
            var IsCorrect: logic = true
            var MinPQCopy:priority_queue(int) = MinPQ

            # Checking the order of top after multiple pops.
            for:
                Index := 0..MinPQ.Size()
                MinPQEl := MinPQCopy.Top[]
                NewMinPQ := MinPQCopy.Pop[IntGreater]
            do:
                case(Index):
                    0 =>
                        if (not MinPQEl = 1) {set IsCorrect = false}
                    1 => 
                        if (not MinPQEl = 2) {set IsCorrect = false}
                    2 => 
                        if (not MinPQEl = 3) {set IsCorrect = false}
                    3 => 
                        if (not MinPQEl = 4) {set IsCorrect = false}
                    4 => 
                        if (not MinPQEl = 5) {set IsCorrect = false}
                    5 => 
                        if (not MinPQEl = 6) {set IsCorrect = false}
                    _ =>
                        Print("In default case for TestMinPQTopAndPop. This should never happen.")

                set MinPQCopy = NewMinPQ
            
            TestName<localizes>:message = "TestMinPQTopAndPop"
            PrintIsCorrectMessage(IsCorrect, TestName)


        # Test Push function on a Min PQ.
        # Order before push: {22,55,77,77,99,79,89}.
        # Expected order after push: {2,22,77,55,99,79,89,77}.
        TestMinPush():void=
            Array := {22,55,77,77,99,79,89}
            var IsCorrect: logic = true
            var PQ: priority_queue(int) = priority_queue(int){}.Initialize(Array, IntGreater)

            set PQ = PQ.Push(2, IntGreater)

            if (not PQ.Size() = 8):
                Print("PQPush size incorrect.")
                set IsCorrect = false

            if:
                not PQ.Elements[0] = 2
                not PQ.Elements[1] = 22
                not PQ.Elements[2] = 77
                not PQ.Elements[3] = 55
                not PQ.Elements[4] = 99
                not PQ.Elements[5] = 79
                not PQ.Elements[6] = 89
                not PQ.Elements[7] = 77
            then:
                Print("PQ elements incorrect after push.")
                set IsCorrect = false

            if (PQ.Empty[]):
                Print("PQPush empty incorrect.")
                set IsCorrect = false
            
            TestName<localizes>:message = "TestMinPush"
            PrintIsCorrectMessage(IsCorrect, TestName)


        # Simulate players leaving and joining a queue.
        #   The player's number should be entered back into the queue to be given out to a different player.
        TestPlayersLeavingAndJoining():void=
            Array := array{}
            var IsCorrect: logic = true
            var PQ: priority_queue(int) = priority_queue(int){}.Initialize(Array, IntGreater)

            set PQ = PQ.Push(1, IntGreater)
            set PQ = PQ.Push(2, IntGreater)
            set PQ = PQ.Push(3, IntGreater)
            set PQ = PQ.Push(4, IntGreater)

            # Size check 1.
            if (not PQ.Size() = 4):
                Print("PlayersLeavingAndJoining check 1 size incorrect.")
                set IsCorrect = false

            # Element value check 1.
            if:
                not PQ.Elements[0] = 1
                not PQ.Elements[1] = 2
                not PQ.Elements[2] = 3
                not PQ.Elements[3] = 4
            then:
                Print("PQ elements check 1 incorrect after push.")
                set IsCorrect = false

            # Empty check 1.
            if (PQ.Empty[]):
                Print("PQPush empty incorrect.")
                set IsCorrect = false

            # Queue 3 players, have player 2 leave.
            var PlayerNumsChosen: []int = array{}
            for:
                Index := 0..2
                PQElement := PQ.Top[]
                NewPQ := PQ.Pop[IntGreater]
            do:
                set PQ = NewPQ
                set PlayerNumsChosen += array{PQElement}

            # Check that the array is {1, 2, 3}.
            for (Index -> Element : PlayerNumsChosen):
                if (not (Index + 1) = Element):
                    Print("PlayersLeavingAndJoining PlayerNumsChosen array incorrect.")
                    set IsCorrect = false

            # Size check 2.
            if (not PQ.Size() = 1):
                Print("PlayersLeavingAndJoining check 2 size incorrect.")
                set IsCorrect = false

            # Element value check 2.
            if:
                not PQ.Elements[0] = 4
            then:
                Print("PlayersLeavingAndJoining PQ elements check 2 incorrect after pop.")
                set IsCorrect = false

            set PQ = PQ.Push(2, IntGreater)

            # Size check 3.
            if (not PQ.Size() = 2):
                Print("PlayersLeavingAndJoining check 3 size incorrect.")
                set IsCorrect = false

            # Element value after push check 3.
            if:
                not PQ.Elements[0] = 2
                not PQ.Elements[1] = 4
            then:
                Print("PlayersLeavingAndJoining PQ elements check 3 incorrect after push.")
                set IsCorrect = false
            
            TestName<localizes>:message = "PlayersLeavingAndJoining"
            PrintIsCorrectMessage(IsCorrect, TestName)


        # === Compare Functors ===

        IntLess<private>(InInt1:int, InInt2:int)<computes><decides>:void=
            InInt1 < InInt2


        IntGreater<private>(InInt1:int, InInt2:int)<computes><decides>:void=
            InInt1 > InInt2


        FloatLess<private>(InFloat1:float, InFloat2:float)<computes><decides>:void=
            InFloat1 < InFloat2


        FloatGreater<private>(InFloat1:float, InFloat2:float)<computes><decides>:void=
            InFloat1 > InFloat2


        # === Helpers ===

        PrintIntArray(InArray:[]int):void=
            for (Index -> El : InArray):
                Print("Index: {Index}: {El}")

        PrintIsCorrectMessage(IsCorrect:logic, TestName:message):void=
            if (IsCorrect?):
                Print(MakeCorrectMessage(TestName))
            else:
                Print(MakeFailureMessage(TestName))


        MakeCorrectMessage<localizes>(TestName:message):message= "+++++ {TestName} passed."
        MakeFailureMessage<localizes>(TestName:message):message= "----- {TestName} failed."