<#
@UUID: 372de2a4-4458-4b10-9567-b39c6ea66a5a
@Task: Horse Derby
@Description: Horse Derby is a competitive multiplayer game where players compete by controlling a mechanical horse in order to make it jump and crouch while it moves autonomously along a predetermined track, from left to right. Players must `Jump` and `Crouch` in order to maneuver their horse among `HIGH` and `LOW` obstacles. `LOW` obstacles must be jumped over, while `HIGH` obstacles must be crouched under. If a player's horse collides with an obstacle, their horse is stunned for a short period time where no input or movement is registered. Collidiing with obstacles and getting stunned allows competitors a chance to gain a lead. If a horse is stunned, it cannot move, jump, or crouch. The first player-controlled horse to cross the finish line is declared the winner. Horse Derby can result in a tie if two or more players finish at the same time, in which case, players must rematch in order to determine a winner.
@Tags: Minigame, Horses, Horse Derby, Race, Horse Race, Racing, Competitive, Multiplayer, Input Trigger Devices, Obstacles, Race Track, Start Line, Finish Line, Collision, Overlap Events
@Export: true
#>